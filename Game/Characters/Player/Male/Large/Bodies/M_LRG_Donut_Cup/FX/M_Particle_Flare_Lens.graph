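# Additive unlit lens-flare sprite material for Cascade and Niagara particles, filed with the Donut Cup character's FX.
# A flare texture is desaturated, raised to a power, scaled by intensity and particle color, then passed through emissive adjustment functions.
# Particle alpha multiplies a depth fade.
# A separate branch projects the object's center to screen space and compares scene depth with its offset camera depth, yielding a clamped occlusion-style value.
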Material /Game/Characters/Player/Male/Large/Bodies/M_LRG_Donut_Cup/FX/M_Particle_Flare_Lens
graph {
 properties {
  bEnableResponsiveAA=true
  bUsedWithNiagaraSprites=true
  bUsedWithParticleSprites=true
  bUsedWithStaticLighting=true
  blendMode=BLEND_Additive
  pixelDepthOffsetMode=PDOM_Legacy
  shadingModel=MSM_Unlit
  usedShadingModels=Unlit
 }
 nodes [
  n1 = MaterialFunctionCall() {materialFunction=&ScreenResolution}
  n2 = AppendVector(A=Constant() {r=1}, B=Divide(A=ComponentMask(None=@n1.Visible Resolution) {r=true}, B=ComponentMask(None=@n1.Visible Resolution) {g=true}))
  n3 = MaterialFunctionCall() {materialFunction=&ScreenResolution}
  n4 = Divide(A=@n3.Visible Resolution, B=@n3.Visible Resolution)
  n5 = Constant2Vector() {g=-0.5, r=0.5}
  n6 = ComponentMask(None=Multiply(A=ComponentMask(None=@n2) {g=true, r=true}, B=Multiply(A=@n4, B=@n5))) {g=true, r=true}
  n7 = TransformPosition(None=ObjectPositionWS()) {transformSourceType=TRANSFORMPOSSOURCE_World, transformType=TRANSFORMPOSSOURCE_View}
  n8 = Divide(A=ComponentMask(None=@n7) {g=true, r=true}, B=ComponentMask(None=@n7) {b=true})
  n9 = ComponentMask(None=ViewProperty() {property=MEVP_TanHalfFieldOfView}.Property) {r=true}
  n10 = ComponentMask(None=Constant3Vector() {constant=<rgba 0.5 0.5 0 1>}) {g=true, r=true}
  n11 = Add(A=Multiply(A=Multiply(A=@n6, B=@n8), B=Divide(B=@n9)), B=ComponentMask(None=Multiply(A=@n4, B=@n10)) {g=true, r=true})
  n12 = MaterialFunctionCall() {materialFunction=&CameraDirectionVector}
  n13 = ObjectPositionWS()
  n14 = CameraPositionWS()
  n15 = Subtract(A=@n13, B=@n14)
  n16 = Multiply(A=DotProduct(A=Normalize(VectorInput=@n12.Result), B=Normalize(VectorInput=@n15)), B=Distance(A=@n13, B=@n14))
  n17 = ScalarParameter() {defaultValue=-1, parameterName=DepthOffset}
  Clamp(None=Subtract(A=SceneDepth(UVs=@n11).R, B=Subtract(A=@n16, B=@n17)))
  n18 = TextureSampleParameter2D() {parameterName=Flare Texture, texture=/Game/Effects/Fort_Effects/Textures/Flares/T_Flare_Spikeball}
  n19 = ScalarParameter() {parameterName=Desaturation}
  n20 = Multiply(
   A=Power(Base=Desaturation(None=@n18.RGB, Fraction=@n19), Exp=ScalarParameter() {defaultValue=1, parameterName=Power})
   B=ScalarParameter() {defaultValue=10, parameterName=FlareIntensity}
  )
  n21 = ParticleColor()
  MaterialFunctionCall(Emissive=MaterialFunctionCall(Emissive=Multiply(A=@n20, B=@n21.RGB)) {materialFunction=&MF_FX_EmissiveAdjustments}.Result) {materialFunction=&MF_FX_EmissiveAdjustments}
  Multiply(A=DepthFade(Opacity=Constant() {r=1}, FadeDistance=ScalarParameter() {parameterName=FadeDistance}), B=@n21.A)
 ]
}
