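# Physics asset of cape colliders for the large male Donut Cup player body, apparently for the cape to collide against.
# Oversized capsules cover the head, neck, upper spine, pelvis, arms, calves and feet, with a wide flat box on the root acting as a floor.
# Shapes are asymmetric: the right calf has a second capsule, and only the left thigh twist bone has a body.
# Constraints link the bodies, several with free swing and twist.
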PhysicsAsset /Game/Characters/Player/Male/Large/Bodies/M_LRG_Donut_Cup/Meshes/M_LRG_Donut_Cup_Custom_CapeColliders
bodies {
 calf_l [{shapeType=Capsule, length=27.7094, radius=16.2212, center=<16.4867 0.0005 1.4425>, rotation=<rot -84.9996 0.0001 -0.0001>}]
 calf_r [
  {shapeType=Capsule, length=32.7094, radius=11.2212, center=<-16.4867 -0.0005 -1.4425>, rotation=<rot -84.9996 0.0001 179.9999>}
  {shapeType=Capsule, length=6.75, radius=10.75, center=<-27.4288 2.0744 1.6848>, rotation=<rot -84.997 -0.0001 -4.9999>}
 ]
 clavicle_l [{shapeType=Capsule, length=7.9594, radius=14.2212, center=<12.0979 1.3188 0>, rotation=<rot -90 0.0099 -0.0099>}]
 clavicle_r [{shapeType=Capsule, length=6.4594, radius=15.7212, center=<-12.0979 -1.1622 0>, rotation=<rot -90 0.0025 179.9976>}]
 foot_l [{shapeType=Capsule, length=30.7094, radius=13.2212, center=<16.4867 0.0005 1.4425>, rotation=<rot -84.9996 0.0001 -0.0001>}]
 foot_r [{shapeType=Capsule, length=30.7094, radius=13.2212, center=<-16.4867 -0.0005 -1.4425>, rotation=<rot -84.9996 0.0001 179.9999>}]
 hand_l [{shapeType=Capsule, length=6.75, radius=6.7212, center=<5.0767 -0.0009 -0.0006>, rotation=<rot -90 0.0099 -0.0099>}]
 hand_r [{shapeType=Capsule, length=6.75, radius=6.7212, center=<-5.0767 0.0009 0.0006>, rotation=<rot -90 -0.0148 -179.9852>}]
 head [{shapeType=Capsule, length=8.5, radius=13.2212, center=<-0.2819 -0.001 0>, rotation=<rot -90 -0.0033 0.0034>}]
 lowerarm_l [{shapeType=Capsule, length=10.9594, radius=11.2212, center=<12.098 -0.0009 0>, rotation=<rot -90 -0.0066 0.0066>}]
 lowerarm_r [{shapeType=Capsule, length=26.4594, radius=11.2212, center=<-12.098 0.0009 0>, rotation=<rot -90 0.0049 179.9951>}]
 neck_01 [{shapeType=Capsule, length=20.25, radius=11.7212, center=<-0.2819 -0.001 0>, rotation=<rot -90 -0.0033 0.0034>}]
 pelvis [{shapeType=Capsule, length=13, radius=18.7212, center=<1.2665 2.1022 0.0001>, rotation=<rot 0.0002 0 0>}]
 root [{shapeType=Box, extentX=213.51, extentY=189.01, extentZ=8.51, center=<0 0 -3.393>, rotation=<rot 0 -90 0>}]
 spine_03 [{shapeType=Capsule, length=14.25, radius=16.9712, center=<-0.2819 -0.001 0>, rotation=<rot 0.0002 0 0>}]
 spine_04 [{shapeType=Capsule, length=17.25, radius=18.9712, center=<-0.2819 2.2489 -0.0006>, rotation=<rot 0.0002 0 0>}]
 spine_05 [{shapeType=Capsule, length=15.25, radius=16.7212, center=<-0.2818 -0.5132 0.0012>, rotation=<rot 0.0002 0 0>}]
 thigh_twist_01_l [{shapeType=Capsule, length=13.9594, radius=5.2212, center=<17.6537 -2.8822 -9.0016>, rotation=<rot -82.6304 -42.6032 -42.6134>}]
 upperarm_l [{shapeType=Capsule, length=10.9594, radius=11.2212, center=<12.098 -0.0009 0>, rotation=<rot -90 -0.0066 0.0066>}]
 upperarm_r [{shapeType=Capsule, length=9.2094, radius=12.9712, center=<-12.098 0.0009 0>, rotation=<rot -90 0.0049 179.9951>}]
}
constraints [
 {bone1Name=spine_03, bone2Name=pelvis, swing1LimitDegrees=45, swing2LimitDegrees=45, twistLimitDegrees=45}
 {bone1Name=spine_04, bone2Name=spine_03, swing1LimitDegrees=45, swing2LimitDegrees=45, twistLimitDegrees=45}
 {bone1Name=lowerarm_l, bone2Name=upperarm_l, swing1LimitDegrees=45, swing2LimitDegrees=45, twistLimitDegrees=45}
 {bone1Name=hand_l, bone2Name=lowerarm_l, swing1LimitDegrees=45, swing2LimitDegrees=45, twistLimitDegrees=45}
 {bone1Name=calf_l, bone2Name=pelvis, swing1LimitDegrees=45, swing2LimitDegrees=45, twistLimitDegrees=45}
 {bone1Name=foot_l, bone2Name=calf_l, swing1LimitDegrees=45, swing2LimitDegrees=45, twistLimitDegrees=45}
 {bone1Name=hand_r, bone2Name=lowerarm_r, swing1LimitDegrees=45, swing2LimitDegrees=45, twistLimitDegrees=45}
 {
  bone1Name=spine_05
  bone2Name=spine_04
  swing1LimitDegrees=45
  swing1Motion=Free
  swing2LimitDegrees=45
  swing2Motion=Free
  twistLimitDegrees=45
  twistMotion=Free
 }
 {
  bone1Name=upperarm_l
  bone2Name=clavicle_l
  swing1LimitDegrees=45
  swing1Motion=Free
  swing2LimitDegrees=45
  swing2Motion=Free
  twistLimitDegrees=45
  twistMotion=Free
 }
 {
  bone1Name=neck_01
  bone2Name=spine_05
  swing1LimitDegrees=45
  swing1Motion=Free
  swing2LimitDegrees=45
  swing2Motion=Free
  twistLimitDegrees=45
  twistMotion=Free
 }
 {
  bone1Name=head
  bone2Name=neck_01
  swing1LimitDegrees=45
  swing1Motion=Free
  swing2LimitDegrees=45
  swing2Motion=Free
  twistLimitDegrees=45
  twistMotion=Free
 }
 {bone1Name=calf_r, bone2Name=pelvis, swing1LimitDegrees=45, swing2LimitDegrees=45, twistLimitDegrees=45}
 {bone1Name=foot_r, bone2Name=calf_r, swing1LimitDegrees=45, swing2LimitDegrees=45, twistLimitDegrees=45}
 {
  bone1Name=upperarm_r
  bone2Name=clavicle_r
  swing1LimitDegrees=45
  swing1Motion=Free
  swing2LimitDegrees=45
  swing2Motion=Free
  twistLimitDegrees=45
  twistMotion=Free
 }
 {bone1Name=lowerarm_r, bone2Name=upperarm_r, swing1LimitDegrees=45, swing2LimitDegrees=45, twistLimitDegrees=45}
 {bone1Name=clavicle_r, bone2Name=spine_05, swing1LimitDegrees=45, swing2LimitDegrees=45, twistLimitDegrees=45}
 {
  bone1Name=pelvis
  bone2Name=root
  swing1LimitDegrees=45
  swing1Motion=Free
  swing2LimitDegrees=45
  swing2Motion=Free
  twistLimitDegrees=45
  twistMotion=Free
 }
 {
  bone1Name=thigh_twist_01_l
  bone2Name=pelvis
  swing1LimitDegrees=45
  swing1Motion=Free
  swing2LimitDegrees=45
  swing2Motion=Free
  twistLimitDegrees=45
  twistMotion=Free
 }
 {
  bone1Name=clavicle_l
  bone2Name=spine_05
  swing1LimitDegrees=45
  swing1Motion=Free
  swing2LimitDegrees=45
  swing2Motion=Free
  twistLimitDegrees=45
  twistMotion=Free
 }
]
